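; Alarm prop that switches lights, flare and sound between alert and deactivated states, starting forced off.
; AlertState's OnRep applies changes cosmetically after a random delay of up to two seconds, staggering alarms.
; UpdateAlertVisuals swaps light color and light function, toggles audio, and hands lighting back to the time-of-day component when inactive.
; Environment linking logic retries on a timer, and the alarm destroys itself when its linked actor is destroyed.
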(event EventOnWorldReady
  ;[0 1 0.135] EventOnWorldReady
  ; Try to link to the environment when the World is Ready.
  (|CallParentFunction)
  (Networking|SwitchHasAuthority))

(event EventBeginPlay
  ;[1 1 1] Force all props to be off on first load
  (|CallParentFunction)
  (Networking|CanExecuteCosmeticEvents
    (:True
      (Utilities|FlowControl|DoOnce false
        (:Completed
          (CallFunction|UpdateAlertVisuals true))))))

(event Custom|AlertStateChanged
  (params (NewState :type Boolean))
  (Utilities|FlowControl|Delay (Math|Random|RandomFloatinRange 0.0 2.0))
  (CallFunction|UpdateAlertVisuals false NewState))

(event Custom|CustomEvent
  (bind _linkattemptsmax (Variables|LinktoEnvironment|GetLinkAttemptsMax))
  ;[0.3798 0.6987 1] TryToLinkToEnvironment
  ; Attempt to attach the alarm to the environment to prevent it from floating if the env is destroyed
  (bind _returnvalue (LinkToActorComponent|PerformLinkingCheck (Variables|Default|GetFortLinkToActor) "Down" (Variables|LinkToEnvironment|GetEnvironmentLinkedActor)))
  ;;[1 0 0.8947] Fallback handles case where the alarm which is in a content bundle in Asteria is loaded before the environment piece it is trying to attach to (in a level instance) - !Not expected to be an issue outside of PIE!
  (bind _result (Math|Integer|IncrementInt (Variables|LinktoEnvironment|GetLinkAttemptsCount))
    (:%20%20
      (if (not (or _returnvalue (>= (Variables|LinktoEnvironment|GetLinkAttemptsCount) _linkattemptsmax)))
        (Fortnite|FortShippingLog self (Utilities|String|Append "ALARM: Attempt to Link to Environment Failed on attempt #" (Utilities|String|ToString(Integer) _result) " of " (Utilities|String|ToString(Integer) _linkattemptsmax) ". This is not expected outside of PIE"))
        (bind _returnvalue_2 (Utilities|Time|SetTimerbyEvent (EventDispatchers|CreateEvent :!event TryToLinkToEnvironment) (Variables|LinktoEnvironment|GetLinkAttemptTiming))) ; try to link again
        (Variables|LinktoEnvironment|SetTimerH_RetryLinkToEnvironment _returnvalue_2)))))

(event OnLinkedActorDestroyed(FortLinkToActor)
  (params (DamageCauser :type Actor%20Object%20Reference))
  ;[0.4149 0.164 0] OnLinkedActorDestroyed
  ; Set the life span to destroy the alarm when the attached envionment actor is destoyed
  (bind _audio (Variables|Default|GetAudio)
    (:then
      (Audio|Components|Audio|FadeOut _audio 0.1 0.0 "Linear" :!class /Script/Engine.AudioComponent)
      (Components|Activation|Deactivate _audio))
    (:else
      _))
  (Actor|SetLifeSpan 0.2))

(event EventEndPlay
  (params (EndPlayReason :type EEndPlayReason%20Enum :default "Destroyed"))
  ;[1 0.3416 0] EndPlay
  ; Cleanup- clear timers, unbind events, etc.
  (Utilities|FlowControl|Sequence
    (:then_0
      (Utilities|Time|ClearandInvalidateTimerbyHandle (Variables|LinktoEnvironment|GetTimerH_RetryLinkToEnvironment)))
    (:then_1
      (|CallParentFunction EndPlayReason) ; call parent EndPlay last in case of \norder of operations dependencies
    )))

(fn FindBrightestLightTODValue
  (bind _todlightingcomponent (Variables|Default|GetTODLightingComponent))
  (bind _haslightdata (Class|TODLightingComponent|HasLightData _todlightingcomponent))
  (if _haslightdata
    (bind _brightestintensity (Class|TODLightingComponent|GetBrightestLightIntensity _todlightingcomponent))
    (Variables|Default|SetBrightestTODLightIntensity _brightestintensity)))

(fn GetTimeOfDayBlueprintDefaultVariables
  (return (Utilities|Struct|MakeTimeOfDayBlueprintDefaultVariables 0 0.0 (not (Variables|/Script/Engine.SceneComponent|GetbVisible (Variables|Default|GetPointLight))))))

(fn OnRep_AlertState
  (Networking|CanExecuteCosmeticEvents
    (:True
      (CallFunction|AlertStateChanged (Variables|Default|GetAlertState)))
    (:False
      _))
  (return))

(fn TryToLinkToEnvironment)

(fn UpdateAlertVisuals
  (params (ForceOff :type Boolean) (NewState :type Boolean))
  (bind _localactive (LocalVariables|GetLocalActive))
  (if ForceOff
    (LocalVariables|SetLocalActive false)
    (else
      (LocalVariables|SetLocalActive NewState)))
  (if (LocalVariables|GetLocalActive)
    ;[1 1 1] Audio
    (bind _audio (Variables|Default|GetAudio)
      (:then
        (Components|Activation|SetActive _audio true false))
      (:else))
    (else
      (bind _audio_2 (Variables|Default|GetAudio)
        (:then
          (Audio|Components|Audio|FadeOut _audio_2 1.0 0.0 "Linear" :!class /Script/Engine.AudioComponent))
        (:else))))
  (for _array_element (Actor|GetComponentsByClass "/Script/Engine.LightComponent")
    (Rendering|Components|Light|SetLightColor _array_element (select _localactive (Variables|Default|GetLightColorAlert) (Variables|Default|GetLightColorDeactivated)) true)
    (Rendering|Components|Light|SetLightFunctionMaterial _array_element (select _localactive (Variables|Default|GetLockdownLightFunction) 0))
    (if _localactive
      (Rendering|SetVisibility _array_element true false)
      (Rendering|Components|Light|SetIntensity _array_element (Variables|Default|GetBrightestTODLightIntensity))))
  (Components|Activation|SetActive (Variables|Default|GetTODLightingComponent) (not _localactive) false)
  (Utilities|FlowControl|Sequence
    (:then_0
      (Rendering|SetVisibility (Variables|Default|GetFlare) _localactive false)
      (Rendering|SetVisibility (Variables|Default|GetPointLight) _localactive false)
      (Rendering|Components|Light|SetIntensity (Variables|Default|GetPointLight) 25.0)
      (bind _audio_3 (Variables|Default|GetAudio)
        (:then
          (if _localactive
            (Audio|Components|Audio|Play _audio_3 0.0)
            (else
              (Audio|Components|Audio|FadeOut _audio_3 1.0 0.0 "Linear" :!class /Script/Engine.AudioComponent))))
        (:else)))
    (:then_1
      (return _localactive))))

(fn ConstructionScript
  (Utilities|FlowControl|Sequence
    (:then_0
      (Networking|CanExecuteCosmeticEvents
        (:True
          (CallFunction|UpdateAlertVisuals true))))
    (:then_1
      (Class|TODLightingComponent|RunConstruct (Variables|Default|GetTODLightingComponent) (Variables|Default|GetTest%20Day%20Phase))
      (Default|FindBrightestLightTODValue))))
